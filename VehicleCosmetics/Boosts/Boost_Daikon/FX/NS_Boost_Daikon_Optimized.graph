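# Niagara system for the Daikon vehicle boost cosmetic, trailing tumbling meshes, pale-blue mist and smoke, a scarf ribbon and a snow ribbon behind the car.
# Meshes, MistCone001 and Particles_Smoke merge several layers per emitter, using spawn groups to select values and tag particles through VizTag.
# Most layers spawn at offsets from System.FinalPosition, with skeletal socket sampling disabled.
# Game mode ID tunes smoke and ribbons, while the frontend preview flag adjusts mesh rates and colors.
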NiagaraSystem /VehicleCosmetics/Boosts/Boost_Daikon/FX/NS_Boost_Daikon_Optimized
dependencies {
 usedDataInterfaces [
  &NiagaraDataInterfaceCurve
  &NiagaraDataInterfaceVectorCurve
  &NiagaraDataInterfaceSpriteRendererInfo
  &NiagaraDataInterfaceSkeletalMesh
 ]
 usedDynamicInputs []
 usedModules [
  &self:SystemSpawnScript.SystemScriptSource.SystemScriptGraph.NiagaraNodeAssignment_0.SetVariables_#1
  &SystemState
 ]
}
emitters {
 Meshes {
  emitterUpdateScript [
   &EmitterState(
    Inactive Response=::NewEnumerator1
    Life Cycle Mode=::NewEnumerator1
    Loop Behavior=::NewEnumerator1
    Loop Duration Mode=::NewEnumerator1
    UseLoopDelay=false
   )
   &VC_FilteredSocketCount()
   &SpawnRate(
    SpawnRate=&MakeCustomFloatFromBool(Bool=$User.bIsFrontendPreview, True Float=1, False Float=5)
    Use Spawn Probability=false
   )
   &SpawnRate(
    SpawnGroup=1
    SpawnRate=&MakeCustomFloatFromBool(Bool=$User.bIsFrontendPreview, True Float=2, False Float=6)
    Use Spawn Probability=false
   ) {module=SpawnRate001}
   &SpawnRate(
    SpawnGroup=2
    SpawnRate=&MakeCustomFloatFromBool(Bool=$User.bIsFrontendPreview, True Float=2, False Float=6)
    Use Spawn Probability=false
   ) {module=SpawnRate002}
   &SpawnRate(
    SpawnGroup=3
    SpawnRate=&MakeCustomFloatFromBool(Bool=$User.bIsFrontendPreview, True Float=1, False Float=5)
    Use Spawn Probability=false
   ) {module=SpawnRate003}
   &SpawnRate(SpawnGroup=4, SpawnRate=6, Use Spawn Probability=false) {module=SpawnRate004}
  ]
  particleSpawnScript [
   &VC_PaintedVariantsOverride() {enabled=false}
   /SetVariables(Particles.VizTag:NiagaraInt32=&SpawnGroup()) {isSetParameters=true}
   &InitializeParticle(
    Color Mode=::NewEnumerator0
    Lifetime=/VehicleCosmetics/FX/Niagara/Modules/MergeEmitters/VC_FloatBySpawnGroup(
     SpawnGroup or ExecIndex:NiagaraInt32=0
     CustomInt:NiagaraInt32=0
     MaximumSpawnGroups:NiagaraInt32=5
     Float SG 0:NiagaraFloat=&RandomRangeFloat(Minimum=0.6, Maximum=0.9)
     Float SG 1:NiagaraFloat=&RandomRangeFloat(Minimum=0.9, Maximum=1.2)
     Float SG 2:NiagaraFloat=&RandomRangeFloat(Minimum=1.2, Maximum=1.4)
     Float SG 3:NiagaraFloat=&RandomRangeFloat(Minimum=0.8, Maximum=1)
     Float SG 4:NiagaraFloat=&RandomRangeFloat(Minimum=0.8, Maximum=1.2)
     Float SG 5:NiagaraFloat=0
    )
    Mesh Scale Mode=::NewEnumerator3
    Mesh Uniform Scale=&VC_FloatBySpawnGroup(
     SpawnGroup or ExecIndex:NiagaraInt32=0
     CustomInt:NiagaraInt32=0
     MaximumSpawnGroups:NiagaraInt32=5
     Float SG 0:NiagaraFloat=&RandomRangeFloat(Minimum=0.23, Maximum=0.28)
     Float SG 1:NiagaraFloat=&RandomRangeFloat(Minimum=0.4, Maximum=0.75)
     Float SG 2:NiagaraFloat=&RandomRangeFloat(Minimum=0.7, Maximum=0.9)
     Float SG 3:NiagaraFloat=&RandomRangeFloat(Minimum=0.14, Maximum=0.1)
     Float SG 4:NiagaraFloat=&RandomRangeFloat(Minimum=0.1, Maximum=0.07)
     Float SG 5:NiagaraFloat=0
    )
    Position Offset=&AddVectorToPosition(Position=<-50 0 0>, Vector=$System.FinalPosition)
    Position Offset Coordinate Space=::Local
    Sprite Rotation Mode=::NewEnumerator2
    UsePositionOffset=true
    Write Material Random=true
   )
   &PartitionParticles(Number Of Partitions=$Emitter.PortCount) {enabled=false}
   &SkeletalMeshLocation(
    Bone / Socket Sampling Mode=::NewEnumerator5
    MaskSpawnGroup=false
    Mesh Sampling Type=::NewEnumerator6
    Sampled Position Offset=&AddVectorToPosition(Position=<0 0 0>, Vector=$System.FinalPosition)
    Skeletal Mesh=$System.SkeletalMesh
   ) {enabled=false}
   &AddVelocity(Velocity=/VehicleCosmetics/FX/Niagara/Modules/MergeEmitters/VC_VectorBySpawnGroup(
     SpawnGroup or ExecIndex:NiagaraInt32=0
     CustomInt:NiagaraInt32=0
     MaximumSpawnGroups:NiagaraInt32=5
     Vector Group 0:Vector3f=&RandomRangeVector(Minimum=<-600 100 35>, Maximum=<-400 -100 -10>)
     Vector Group 1:Vector3f=&RandomRangeVector(Minimum=<-500 100 35>, Maximum=<-300 -100 -10>)
     Vector Group 2:Vector3f=&RandomRangeVector(Minimum=<-500 100 35>, Maximum=<-300 -100 -10>)
     Vector Group 3:Vector3f=&RandomRangeVector(Minimum=<-700 100 35>, Maximum=<-450 -100 -10>)
     Vector Group 4:Vector3f=&RandomRangeVector(Minimum=<-600 100 35>, Maximum=<-400 -100 -10>)
     Vector Group 5:Vector3f=<0 0 0>
    ))
   &Color(Color=<rgba 0.3 0.3 0.3 1>)
   &ShapeLocation(
    Cylinder Height=67.9548
    Cylinder Radius=41.2267
    MaskSpawnGroup=false
    Override Seed=false
    Shape Primitive=::NewEnumerator1
    Surface Only=false
    UseOwnerScale=true
   )
   &InitialMeshOrientation(
    Mesh Orientation Mode=::NewEnumerator7
    Rotation=&RandomRangeVector(Minimum=<0 0 0>, Maximum=<1 1 1>)
    Use Rotation Vector=true
   )
  ]
  particleUpdateScript [
   &ParticleState()
   &UpdateMeshOrientation(Rotation Vector=&VC_VectorBySpawnGroup(
     SpawnGroup or ExecIndex:NiagaraInt32=0
     CustomInt:NiagaraInt32=0
     MaximumSpawnGroups:NiagaraInt32=5
     Vector Group 0:Vector3f=<0.2 0.2 0.2>
     Vector Group 1:Vector3f=<0.25 0.25 0.25>
     Vector Group 2:Vector3f=<0.3 0.3 0.3>
     Vector Group 3:Vector3f=<0.3 0.3 0.3>
     Vector Group 4:Vector3f=<0.3 0.3 0.3>
     Vector Group 5:Vector3f=<0 0 0>
    )) {module=UpdateMeshOrientation001}
   &InheritSourceMovement(
    Applied Position Falloff Distance=3000
    Applied Position Scale=&VectorFromFloat(Value=&FloatFromCurve(FloatCurve=DataInterface() {propertyValues {Curve {keys [
          {interpMode=RCIM_Cubic, tangentMode=RCTM_Break, time=0, value=1}
          {arriveTangent=-2.9025, interpMode=RCIM_Cubic, tangentMode=RCTM_Break, time=1, value=0}
         ]}}}))
    ApplyForce=false
    ApplyPosition=true
    ApplyPositionFalloff=false
   )
   &ScaleMeshSize(Scale Factor=&VectorFromFloat(Value=&MaskFloatBySpawnGroup(
      Masked Float=&FloatFromCurve(FloatCurve=DataInterface() {propertyValues {Curve {keys [<at 0.3 1>,<at 1 0>]}}})
      Passthrough Float=&FloatFromCurve(FloatCurve=DataInterface() {propertyValues {Curve {keys [<at 0 0.2>,<at 0.25 1.8>,<at 0.75 1.8>,<at 1 0>]}}})
      Spawn Group Mask=4
     )))
   &ScaleVelocity(
    Coordinate Space:ENiagaraCoordinateSpace=::World
    Velocity Scale=&VectorFromCurve(VectorCurve=DataInterface() {propertyValues {XCurve {keys [<at 0 1.25>,<at 1 0.2>]}, YCurve {keys [<at 0 0.5>,<at 1 -2>]}, ZCurve {keys [<at 0 0>,<at 1 0>]}}})
   ) {enabled=false}
   &CurlNoiseForce(
    Bias Noise=false
    Noise Strength=&VC_FloatBySpawnGroup(
     SpawnGroup or ExecIndex:NiagaraInt32=0
     CustomInt:NiagaraInt32=0
     MaximumSpawnGroups:NiagaraInt32=5
     Float SG 0:NiagaraFloat=356.8454
     Float SG 1:NiagaraFloat=&RandomRangeFloat(Minimum=350, Maximum=150)
     Float SG 2:NiagaraFloat=&RandomRangeFloat(Minimum=350, Maximum=150)
     Float SG 3:NiagaraFloat=&RandomRangeFloat(Minimum=600, Maximum=300)
     Float SG 4:NiagaraFloat=&RandomRangeFloat(Minimum=600, Maximum=300)
     Float SG 5:NiagaraFloat=0
    )
    Pan Noise=false
   )
   &Drag(
    Drag=&FloatFromCurve(FloatCurve=DataInterface() {propertyValues {Curve {keys [<at 0 0>,<at 1 3>]}}})
    Use Linear Drag=true
    Use Rotational Drag=true
   )
   &SolveForcesAndVelocity(Clamp Velocity=false, Limit Acceleration=false)
  ]
  rendererClasses [&NiagaraMeshRendererProperties, &NiagaraSpriteRendererProperties]
  renderers [
   {rendererClass=&NiagaraMeshRendererProperties, rendererIndex=0}
   {rendererClass=&NiagaraMeshRendererProperties, rendererIndex=1}
   {rendererClass=&NiagaraMeshRendererProperties, rendererIndex=2}
   {rendererClass=&NiagaraMeshRendererProperties, rendererIndex=3}
   {rendererClass=&NiagaraMeshRendererProperties, rendererIndex=4}
   {rendererClass=&NiagaraSpriteRendererProperties, rendererIndex=5}
  ]
 }
 MistCone001 {
  emitterUpdateScript [
   &EmitterState(Inactive Response=::NewEnumerator1, Life Cycle Mode=::NewEnumerator1, Loop Duration=1, UseLoopDelay=false)
   &VC_FilteredSocketCount()
   &SpawnRate(SpawnRate=40, Use Spawn Probability=false)
   &SpawnRate(SpawnGroup=1, SpawnRate=40, Use Spawn Probability=false) {module=SpawnRate001}
   &SpawnRate(SpawnGroup=2, SpawnRate=50, Use Spawn Probability=false) {module=SpawnRate002}
   &SpawnBurst_Instantaneous(Spawn Group=3, Use Spawn Probability=false, UseLoopCountLimit=false)
  ]
  particleSpawnScript [
   &VC_PaintedVariantsOverride() {enabled=false}
   /SetVariables(Particles.VizTag:NiagaraInt32=&SpawnGroup()) {isSetParameters=true}
   &InitializeParticle(
    Color Mode=::NewEnumerator0
    Lifetime=&VC_FloatBySpawnGroup(
     SpawnGroup or ExecIndex:NiagaraInt32=0
     CustomInt:NiagaraInt32=0
     MaximumSpawnGroups:NiagaraInt32=4
     Float SG 0:NiagaraFloat=&RandomRangeFloat(Minimum=0.5, Maximum=0.6)
     Float SG 1:NiagaraFloat=&RandomRangeFloat(Minimum=0.8, Maximum=1)
     Float SG 2:NiagaraFloat=&RandomRangeFloat(Minimum=0.3, Maximum=0.6)
     Float SG 3:NiagaraFloat=$Emitter.CurrentLoopDuration
     Float SG 4:NiagaraFloat=0
     Float SG 5:NiagaraFloat=0
    )
    Mesh Scale Mode=::NewEnumerator3
    Mesh Uniform Scale=&MaskFloatBySpawnGroup(Masked Float=&RandomRangeFloat(Minimum=0.03, Maximum=0.02), Passthrough Float=0.23, Spawn Group Mask=2)
    Position Offset=&AddVectorToPosition(
     Position=&MaskVectorBySpawnGroup(Masked Vector=<-50 0 0>, Passthrough Vector=<-80 0 0>, Spawn Group Mask=3)
     Vector=$System.FinalPosition
    )
    Position Offset Coordinate Space=::Local
    Sprite Rotation Mode=::NewEnumerator2
    Sprite Size Mode=::NewEnumerator3
    Uniform Sprite Size=&VC_FloatBySpawnGroup(
     SpawnGroup or ExecIndex:NiagaraInt32=0
     CustomInt:NiagaraInt32=0
     MaximumSpawnGroups:NiagaraInt32=2
     Float SG 0:NiagaraFloat=&RandomRangeFloat(Minimum=20, Maximum=25)
     Float SG 1:NiagaraFloat=&RandomRangeFloat(Minimum=13, Maximum=18)
     Float SG 2:NiagaraFloat=&RandomRangeFloat(Minimum=0.03, Maximum=0.02)
     Float SG 3:NiagaraFloat=0
     Float SG 4:NiagaraFloat=0
     Float SG 5:NiagaraFloat=0
    )
    UsePositionOffset=true
    Write Material Random=true
   )
   &PartitionParticles(Number Of Partitions=$Emitter.PortCount) {enabled=false}
   &SkeletalMeshLocation(
    Bone / Socket Sampling Mode=::NewEnumerator5
    MaskSpawnGroup=false
    Mesh Sampling Type=::NewEnumerator6
    Sampled Position Offset=<-20 0 0>
    Skeletal Mesh=$System.SkeletalMesh
   ) {enabled=false}
   &AddVelocity(Velocity=&VC_VectorBySpawnGroup(
     SpawnGroup or ExecIndex:NiagaraInt32=0
     CustomInt:NiagaraInt32=0
     MaximumSpawnGroups:NiagaraInt32=4
     Vector Group 0:Vector3f=&RandomRangeVector(Minimum=<-550 50 25>, Maximum=<-500 -50 10>)
     Vector Group 1:Vector3f=&RandomRangeVector(Minimum=<-550 50 25>, Maximum=<-500 -50 10>)
     Vector Group 2:Vector3f=&RandomRangeVector(Minimum=<-300 50 50>, Maximum=<-300 -50 90>)
     Vector Group 3:Vector3f=<0 0 0>
     Vector Group 4:Vector3f=<0 0 0>
     Vector Group 5:Vector3f=<0 0 0>
    ))
   &Color(
    Color=/VehicleCosmetics/FX/Niagara/Modules/MergeEmitters/VC_Dynamic_ColorBySpawnGroup(
     Color SG 0:LinearColor=&MakeCustomLinearColorFromBool(Bool=$User.bIsFrontendPreview, True Color=<rgba 0.6302 0.8862 1 1>, False Color=<rgba 0.5 0.8 1 1>)
     MaximumSpawnGroups:NiagaraInt32=4
     Color SG 1:LinearColor=&MakeCustomLinearColorFromBool(Bool=$User.bIsFrontendPreview, True Color=<rgba 0.6302 0.8862 1 1>, False Color=<rgba 0.5 0.8 1 1>)
     Color SG 2:LinearColor=<rgba 0.5 0.5 0.5 1>
     Color SG 3:LinearColor=<rgba 0.4 0.954 1.1 1>
     Color SG 4:LinearColor=<rgba 1 1 1 1>
    )
    Scale Alpha=&VC_FloatBySpawnGroup(
     SpawnGroup or ExecIndex:NiagaraInt32=0
     CustomInt:NiagaraInt32=0
     MaximumSpawnGroups:NiagaraInt32=4
     Float SG 0:NiagaraFloat=0.15
     Float SG 1:NiagaraFloat=5
     Float SG 2:NiagaraFloat=1
     Float SG 3:NiagaraFloat=1
     Float SG 4:NiagaraFloat=0
     Float SG 5:NiagaraFloat=0
    )
    Scale Color=&VC_VectorBySpawnGroup(
     SpawnGroup or ExecIndex:NiagaraInt32=0
     CustomInt:NiagaraInt32=0
     MaximumSpawnGroups:NiagaraInt32=4
     Vector Group 0:Vector3f=<2 2 2>
     Vector Group 1:Vector3f=<1 1 1>
     Vector Group 2:Vector3f=<1 1 1>
     Vector Group 3:Vector3f=<1 1 1>
     Vector Group 4:Vector3f=<0 0 0>
     Vector Group 5:Vector3f=<0 0 0>
    )
   )
   &ShapeLocation(
    MaskSpawnGroup=false
    Override Seed=false
    Sphere Radius=&VC_FloatBySpawnGroup(
     SpawnGroup or ExecIndex:NiagaraInt32=0
     CustomInt:NiagaraInt32=0
     MaximumSpawnGroups:NiagaraInt32=4
     Float SG 0:NiagaraFloat=0
     Float SG 1:NiagaraFloat=0
     Float SG 2:NiagaraFloat=22.0376
     Float SG 3:NiagaraFloat=0
     Float SG 4:NiagaraFloat=0
     Float SG 5:NiagaraFloat=0
    )
    UseOwnerScale=true
   )
   &InitialMeshOrientation(
    Mesh Orientation Mode=::NewEnumerator7
    Rotation=&MaskVectorBySpawnGroup(
     Masked Vector=&RandomRangeVector(Minimum=<0 0 0>, Maximum=<1 1 1>)
     Passthrough Vector=<0 0 0>
     Spawn Group Mask=2
    )
    Use Rotation Vector=true
   )
  ]
  particleUpdateScript [
   &ParticleState()
   &ScaleMeshSize(Scale Factor=&VectorFromFloat(Value=&VC_FloatBySpawnGroup(
      SpawnGroup or ExecIndex:NiagaraInt32=0
      CustomInt:NiagaraInt32=0
      MaximumSpawnGroups:NiagaraInt32=4
      Float SG 0:NiagaraFloat=1
      Float SG 1:NiagaraFloat=1
      Float SG 2:NiagaraFloat=&MakeFloatFromVector(
       Vector=&VectorFromFloat(Value=&FloatFromCurve(FloatCurve=DataInterface() {propertyValues {Curve {keys [<at 0.3 1>,<at 0.95 0>]}}}))
       Channel=::NewEnumerator0
      )
      Float SG 3:NiagaraFloat=1
      Float SG 4:NiagaraFloat=0
      Float SG 5:NiagaraFloat=0
     )))
   &ScaleSpriteSize(
    Scale Sprite Size Mode=::NewEnumerator0
    Uniform Scale Factor=&VC_FloatBySpawnGroup(
     SpawnGroup or ExecIndex:NiagaraInt32=0
     CustomInt:NiagaraInt32=0
     MaximumSpawnGroups:NiagaraInt32=4
     Float SG 0:NiagaraFloat=&FloatFromCurve(FloatCurve=DataInterface() {propertyValues {Curve {keys [{arriveTangent=0.63, interpMode=RCIM_Cubic, leaveTangent=0.63, tangentMode=RCTM_User, time=0, value=1.5}, <at 1 1.8 RCIM_Cubic>]}}})
     Float SG 1:NiagaraFloat=&FloatFromCurve(FloatCurve=DataInterface() {propertyValues {Curve {keys [{arriveTangent=0.63, time=0, value=3}, <at 1 2.3 RCIM_Cubic>]}}})
     Float SG 2:NiagaraFloat=1
     Float SG 3:NiagaraFloat=1
     Float SG 4:NiagaraFloat=0
     Float SG 5:NiagaraFloat=0
    )
   )
   &ScaleVelocity(
    Coordinate Space:ENiagaraCoordinateSpace=::Simulation
    Velocity Scale=&VC_VectorBySpawnGroup(
     SpawnGroup or ExecIndex:NiagaraInt32=0
     CustomInt:NiagaraInt32=0
     MaximumSpawnGroups:NiagaraInt32=4
     Vector Group 0:Vector3f=&VectorFromFloat(Value=&FloatFromCurve(FloatCurve=DataInterface() {propertyValues {Curve {keys [<at 0 0.35>,<at 1 0.2>]}}}))
     Vector Group 1:Vector3f=&VectorFromFloat(Value=&FloatFromCurve(FloatCurve=DataInterface() {propertyValues {Curve {keys [<at 0 0.3>,<at 1 0.2>]}}}))
     Vector Group 2:Vector3f=&VectorFromCurve(VectorCurve=DataInterface() {propertyValues {XCurve {keys [<at 0 1.25>,<at 0.25 1.25>,<at 1 0.2>]}, YCurve {keys [<at 0 1.5>,<at 0.25 1>,<at 1 -2>]}, ZCurve {keys [<at 0 0>,<at 1 0>]}}})
     Vector Group 3:Vector3f=<0 0 0>
     Vector Group 4:Vector3f=<0 0 0>
     Vector Group 5:Vector3f=<0 0 0>
    )
   )
   &SolveForcesAndVelocity(Clamp Velocity=false, Limit Acceleration=false)
   &UpdateMeshOrientation(Rotation Vector=&MaskVectorBySpawnGroup(Masked Vector=<0.3 0.3 0.3>, Passthrough Vector=<0 1.5 0>, Spawn Group Mask=2)) {module=UpdateMeshOrientation001}
   &ScaleColor(
    Scale Alpha=&VC_FloatBySpawnGroup(
     SpawnGroup or ExecIndex:NiagaraInt32=0
     CustomInt:NiagaraInt32=0
     MaximumSpawnGroups:NiagaraInt32=4
     Float SG 0:NiagaraFloat=&FloatFromCurve(FloatCurve=DataInterface() {propertyValues {Curve {keys [<at 0 0.5>,<at 1 0>]}}})
     Float SG 1:NiagaraFloat=&FloatFromCurve(FloatCurve=DataInterface() {propertyValues {Curve {keys [<at 0 1>,<at 1 0>]}}})
     Float SG 2:NiagaraFloat=1
     Float SG 3:NiagaraFloat=1
     Float SG 4:NiagaraFloat=0
     Float SG 5:NiagaraFloat=0
    )
    Scale RGB=&VC_VectorBySpawnGroup(
     SpawnGroup or ExecIndex:NiagaraInt32=0
     CustomInt:NiagaraInt32=0
     MaximumSpawnGroups:NiagaraInt32=4
     Vector Group 0:Vector3f=<1 1 1>
     Vector Group 1:Vector3f=<64 64 64>
     Vector Group 2:Vector3f=<1 1 1>
     Vector Group 3:Vector3f=<1 1 1>
     Vector Group 4:Vector3f=<0 0 0>
     Vector Group 5:Vector3f=<0 0 0>
    )
    ScaleA=true
    ScaleRGB=true
   )
   &DynamicMaterialParameters(
    Index 0 Param 1=&FloatFromCurve(FloatCurve=DataInterface() {propertyValues {Curve {keys [<at 0 1>,<at 0.5 1.5>,<at 1 0>]}}})
    Index 0 Param 3=&FloatFromCurve(FloatCurve=DataInterface() {propertyValues {Curve {keys [<at 0 0>,<at 0.8 0.5>,<at 1 1>]}}})
    Module.Index0Float0Write=true
    Module.Index0Float1Write=true
    Module.Index0Float2Write=true
    Module.Index0Float3Write=true
   )
   &SubUVAnimation(
    Direct Frame Index:NiagaraFloat=&FloatFromCurve(FloatCurve=DataInterface() {propertyValues {Curve {keys [<at 0 32>,<at 1 64>]}}})
    "Sprite Renderer "=DataInterface() {propertyValues {SpriteRenderer=&self:MistCone001_0.NiagaraSpriteRendererProperties_1}}
    SubUV Animation Mode:ENiagara_SubUVLookupModeV2=::NewEnumerator5
   )
  ]
  rendererClasses [&NiagaraSpriteRendererProperties, &NiagaraMeshRendererProperties]
  renderers [
   {rendererClass=&NiagaraSpriteRendererProperties}
   {rendererClass=&NiagaraMeshRendererProperties, rendererIndex=1}
   {rendererClass=&NiagaraMeshRendererProperties, rendererIndex=2}
   {rendererClass=&NiagaraSpriteRendererProperties, rendererIndex=3}
  ]
 }
 Particles_Smoke {
  emitterSpawnScript [/SetVariables(Emitter.Speed:NiagaraFloat=&RandomRangeFloat(Minimum=2, Maximum=3)) {isSetParameters=true}]
  emitterUpdateScript [
   &EmitterState(Inactive Response=::NewEnumerator1, Life Cycle Mode=::NewEnumerator1, Loop Duration=1, UseLoopDelay=false)
   &VC_FilteredSocketCount()
   &SpawnRate(
    SpawnRate=/VehicleCosmetics/FX/Niagara/Modules/GameMode/VC_SetFloatByGameModeID(
     SameValueForLockerAndBR?:NiagaraBool=true
     Locker Float:NiagaraFloat=30
     Battle Royale Float:NiagaraFloat=30
     Rocket Racing Float:NiagaraFloat=60
     GameModeID:NiagaraInt32=$System.GameModeID
    )
    Use Spawn Probability=false
   )
   &SpawnRate(
    SpawnGroup=1
    SpawnRate=&VC_SetFloatByGameModeID(
     SameValueForLockerAndBR?:NiagaraBool=false
     Locker Float:NiagaraFloat=25
     Battle Royale Float:NiagaraFloat=25
     Rocket Racing Float:NiagaraFloat=40
     GameModeID:NiagaraInt32=$System.GameModeID
    )
    Use Spawn Probability=false
   ) {module=SpawnRate001}
  ]
  particleSpawnScript [
   &VC_PaintedVariantsOverride() {enabled=false}
   /SetVariables(Particles.VizTag:NiagaraInt32=&SpawnGroup()) {isSetParameters=true}
   &InitializeParticle(
    Color Mode=::NewEnumerator0
    Lifetime=&MaskFloatBySpawnGroup(
     Masked Float=&RandomRangeFloat(Minimum=0.8, Maximum=0.85)
     Passthrough Float=&VC_SetFloatByGameModeID(
      SameValueForLockerAndBR?:NiagaraBool=false
      Locker Float:NiagaraFloat=&RandomRangeFloat(Minimum=0.654, Maximum=1.2)
      Battle Royale Float:NiagaraFloat=&RandomRangeFloat(Minimum=0.654, Maximum=1.2)
      Rocket Racing Float:NiagaraFloat=&RandomRangeFloat(Minimum=0.654, Maximum=0.8)
      GameModeID:NiagaraInt32=$System.GameModeID
     )
    )
    Position Offset=&AddVectorToPosition(Position=<-50 0 0>, Vector=$System.FinalPosition)
    Position Offset Coordinate Space=::Local
    Sprite Rotation Angle=&RandomRangeFloat(Minimum=0, Maximum=1)
    Sprite Rotation Mode=::NewEnumerator1
    Sprite Size Mode=::NewEnumerator3
    Uniform Sprite Size=&MaskFloatBySpawnGroup(
     Masked Float=&RandomRangeFloat(Minimum=180, Maximum=200)
     Passthrough Float=&RandomRangeFloat(Minimum=15, Maximum=20)
    )
    UsePositionOffset=true
    Write Material Random=true
   )
   &PartitionParticles(Number Of Partitions=$Emitter.PortCount) {enabled=false}
   &SkeletalMeshLocation(
    Bone / Socket Sampling Mode=::NewEnumerator5
    MaskSpawnGroup=false
    Mesh Sampling Type=::NewEnumerator6
    Sampled Position Offset=<-20 0 0>
    Skeletal Mesh=$System.SkeletalMesh
   ) {enabled=false}
   &AddVelocity(Velocity=&MaskVectorBySpawnGroup(
     Masked Vector=<-1500 0 0>
     Passthrough Vector=&RandomRangeVector(Minimum=<-1800 40 0>, Maximum=<-1800 -40 0>)
    ))
   &Color(
    Color=&MakeCustomLinearColorFromBool(Bool=$User.bIsFrontendPreview, True Color=<rgba 0.6302 0.8862 1 1>, False Color=<rgba 0.5 0.8 1 1>)
    Scale Alpha=&MaskFloatBySpawnGroup(Masked Float=1, Passthrough Float=5)
   )
  ]
  particleUpdateScript [
   &ParticleState()
   &ScaleSpriteSize(
    Scale Sprite Size Mode=::NewEnumerator0
    Uniform Scale Factor=&MaskFloatBySpawnGroup(
     Masked Float=&FloatFromCurve(FloatCurve=DataInterface() {propertyValues {Curve {keys [
          {arriveTangent=0.4581, interpMode=RCIM_Cubic, leaveTangent=0.4581, tangentMode=RCTM_User, time=0.025, value=0.25}
          {arriveTangent=0.0472, interpMode=RCIM_Cubic, leaveTangent=0.0472, tangentMode=RCTM_User, time=0.5, value=0.7}
          <at 1 0.7>
         ]}}})
     Passthrough Float=&FloatFromCurve(FloatCurve=DataInterface() {propertyValues {Curve {keys [
          {arriveTangent=12.6863, interpMode=RCIM_Cubic, leaveTangent=12.6863, tangentMode=RCTM_User, time=0, value=3}
          {arriveTangent=36.8417, interpMode=RCIM_Cubic, leaveTangent=36.8417, tangentMode=RCTM_User, time=0.6, value=13}
         ]}}})
    )
   )
   &ScaleVelocity(
    Coordinate Space:ENiagaraCoordinateSpace=::Simulation
    Velocity Scale=&VectorFromFloat(Value=&MaskFloatBySpawnGroup(
      Masked Float=&FloatFromCurve(FloatCurve=DataInterface() {propertyValues {Curve {keys [<at 0.5 0.45>,<at 0.7 1>,<at 1 1>]}}})
      Passthrough Float=0.25
     ))
   )
   &CurlNoiseForce(
    Bias Noise=false
    Noise Strength=&MaskFloatBySpawnGroup(Masked Float=350, Passthrough Float=0)
    Pan Noise=false
   )
   &InheritSourceMovement(
    Applied Position Falloff Distance=3000
    Applied Position Scale=&VectorFromFloat(Value=&FloatFromCurve(FloatCurve=DataInterface() {propertyValues {Curve {keys [
          {interpMode=RCIM_Cubic, tangentMode=RCTM_Break, time=0, value=1}
          {arriveTangent=-2.9025, interpMode=RCIM_Cubic, tangentMode=RCTM_Break, time=1, value=0}
         ]}}}))
    ApplyForce=false
    ApplyPosition=true
    ApplyPositionFalloff=false
   )
   &SolveForcesAndVelocity(Clamp Velocity=false, Limit Acceleration=false)
   &ScaleColor(
    Scale Alpha=&MaskFloatBySpawnGroup(
     Masked Float=&FloatFromCurve(FloatCurve=DataInterface() {propertyValues {Curve {keys [<at 0 0>,<at 0.097 1>,<at 0.5 0.5>,<at 0.7 0>]}}})
     Passthrough Float=&FloatFromCurve(FloatCurve=DataInterface() {propertyValues {Curve {keys [<at 0 1>,<at 1 0>]}}})
    )
    Scale RGB=&VectorFromFloat(Value=&MaskFloatBySpawnGroup(
      Masked Float=&VC_SetFloatByGameModeID(
       SameValueForLockerAndBR?:NiagaraBool=false
       Locker Float:NiagaraFloat=0.8
       Battle Royale Float:NiagaraFloat=1
       Rocket Racing Float:NiagaraFloat=1.1
       GameModeID:NiagaraInt32=$System.GameModeID
      )
      Passthrough Float=&VC_SetFloatByGameModeID(
       SameValueForLockerAndBR?:NiagaraBool=true
       Locker Float:NiagaraFloat=64
       Battle Royale Float:NiagaraFloat=64
       Rocket Racing Float:NiagaraFloat=128
       GameModeID:NiagaraInt32=$System.GameModeID
      )
     ))
    ScaleA=true
    ScaleRGB=true
   )
  ]
  rendererClasses [&NiagaraSpriteRendererProperties]
  renderers [
   {rendererClass=&NiagaraSpriteRendererProperties}
   {rendererClass=&NiagaraSpriteRendererProperties, rendererIndex=1}
  ]
 }
 Scarf {
  emitterUpdateScript [
   &EmitterState()
   &VC_FilteredSocketCount()
   &VC_RateQualityHandler()
   &VC_SpawnRibbon_v1(SpawnRate=25)
  ]
  particleSpawnScript [
   &VC_PaintedVariantsOverride() {enabled=false}
   &VC_InitializeRibbon_v1(
    Custom Offset=true
    Facing Axis:ENiagaraOrientationAxis=::YAxis
    Lifetime=&VC_SetFloatByGameModeID(
     SameValueForLockerAndBR?:NiagaraBool=false
     Locker Float:NiagaraFloat=1.1
     Battle Royale Float:NiagaraFloat=1.1
     Rocket Racing Float:NiagaraFloat=0.8
     GameModeID:NiagaraInt32=$System.GameModeID
    )
    Local Vector Offset=/VehicleCosmetics/FX/Niagara/Modules/GameMode/VC_SetVectorByGameModeID(
     SameValueForLockerAndBR?:NiagaraBool=true
     Locker Vector:Vector3f=<-260 0 30>
     Battle Royale Vector:Vector3f=<0 0 0>
     Rocket Racing Vector:Vector3f=<-250 0 30>
     GameModeID:NiagaraInt32=$System.GameModeID
    )
    Offset Amount=1
    Ribbon Width=25
   )
   &AddVelocity(Velocity=<-800 0 0>) {module=AddVelocity001}
  ]
  particleUpdateScript [
   &ParticleState()
   &InheritSourceMovement(
    Applied Position Falloff Distance=3000
    Applied Position Scale=&VectorFromFloat(Value=&FloatFromCurve(FloatCurve=DataInterface() {propertyValues {Curve {keys [
          {interpMode=RCIM_Cubic, tangentMode=RCTM_Break, time=0, value=1}
          {arriveTangent=-2.9025, interpMode=RCIM_Cubic, tangentMode=RCTM_Break, time=1, value=0}
         ]}}}))
    ApplyForce=false
    ApplyPosition=true
    ApplyPositionFalloff=false
   )
   &SolveForcesAndVelocity(Clamp Velocity=false, Limit Acceleration=false)
   &VC_UpdateRibbon_v1()
   &ScaleColor(
    Scale Alpha=&FloatFromCurve(FloatCurve=DataInterface() {propertyValues {Curve {keys [<at 0 1>, {arriveTangent=-1.087, interpMode=RCIM_Cubic, leaveTangent=-1.087, time=0.89, value=1}, <at 0.92 0 RCIM_Cubic>]}}})
    ScaleA=false
    ScaleRGB=true
   )
  ]
  rendererClasses [&NiagaraRibbonRendererProperties]
  renderers [{rendererClass=&NiagaraRibbonRendererProperties}]
 }
 SnowRibbon {
  emitterUpdateScript [
   &EmitterState()
   &VC_FilteredSocketCount()
   &VC_RateQualityHandler()
   &VC_SpawnRibbon_v1(SpawnRate=$Emitter.Spawn Rate)
  ]
  particleSpawnScript [
   &VC_PaintedVariantsOverride() {enabled=false}
   &VC_InitializeRibbon_v1(
    Color=<rgba 0.6302 0.8862 1 0.45>
    Custom Offset=true
    Lifetime=&VC_SetFloatByGameModeID(
     SameValueForLockerAndBR?:NiagaraBool=false
     Locker Float:NiagaraFloat=1.5
     Battle Royale Float:NiagaraFloat=1.5
     Rocket Racing Float:NiagaraFloat=1.2
     GameModeID:NiagaraInt32=$System.GameModeID
    )
    Local Vector Offset=&AddVectorToPosition(Position=<-81 0 0>, Vector=$System.FinalPosition)
    Offset Amount=1
    Ribbon Width=120
   )
   &AddVelocity(Velocity=<-800 0 0>) {module=AddVelocity001}
  ]
  particleUpdateScript [
   &ParticleState()
   &InheritSourceMovement(
    Applied Position Falloff Distance=3000
    Applied Position Scale=&VectorFromFloat(Value=&FloatFromCurve(FloatCurve=DataInterface() {propertyValues {Curve {keys [
          {interpMode=RCIM_Cubic, tangentMode=RCTM_Break, time=0, value=1}
          {arriveTangent=-2.9025, interpMode=RCIM_Cubic, tangentMode=RCTM_Break, time=1, value=0}
         ]}}}))
    ApplyForce=false
    ApplyPosition=true
    ApplyPositionFalloff=false
   )
   &SolveForcesAndVelocity(Clamp Velocity=false, Limit Acceleration=false)
   &VC_UpdateRibbon_v1()
   &ScaleColor(
    Scale Alpha=&FloatFromCurve(FloatCurve=DataInterface() {propertyValues {Curve {keys [<at 0 0.15>,<at 1 0 RCIM_Cubic>]}}})
    Scale RGB=<7.5 7.5 7.5>
    ScaleA=true
    ScaleRGB=true
   )
   &ScaleRibbonWidth(Ribbon Width Scale=&FloatFromCurve(FloatCurve=DataInterface() {propertyValues {Curve {keys [<at 0 0>,<at 0.1 1>,<at 1 3.5>]}}}))
  ]
  rendererClasses [&NiagaraRibbonRendererProperties]
  renderers [{rendererClass=&NiagaraRibbonRendererProperties}]
 }
}
summary {
 systemName=NS_Boost_Daikon_Optimized
 emitters [
  {emitterName=Scarf, rendererClasses [&NiagaraRibbonRendererProperties]}
  {emitterName=SnowRibbon, rendererClasses [&NiagaraRibbonRendererProperties]}
  {emitterName=Particles_Smoke, rendererClasses [&NiagaraSpriteRendererProperties]}
  {emitterName=Meshes, rendererClasses [&NiagaraMeshRendererProperties, &NiagaraSpriteRendererProperties]}
  {emitterName=MistCone001, rendererClasses [&NiagaraSpriteRendererProperties, &NiagaraMeshRendererProperties]}
 ]
 userVariables [
  User.bIsFrontendPreview {type=NiagaraBool}
  User.bIsPainted {type=NiagaraBool}
  User.ColorVariantID {type=NiagaraInt32}
  User.DTCarOffset {type=Vector3f, description="", defaultValue {struct=Vector3f, value=<0 0 0>}}
  User.GameModeID {type=NiagaraInt32}
  User.PaintedColor {type=LinearColor, description="", defaultValue {struct=LinearColor, value=<rgba 1 1 1 1>}}
  User.PaintedScalar {type=NiagaraFloat, defaultValue=1, description=""}
  User.PortCount {type=NiagaraInt32, defaultValue=1}
  User.SingleBoostPos {type=NiagaraPosition, description="", defaultValue {}}
 ]
}
