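# Masked unlit Niagara-ribbon material for the snowman scarf in the Daikon vehicle boost cosmetic.
# Color combines scarf diffuse and tinted white areas via RGB masks, shaded by a desaturated fabric normal map; particle color tints it, while alpha scales mask-based opacity.
# Time-driven sine and cosine waves build world-position-offset motion: rippling, UV-masked twist and pinch, plus end waving with SplineThicken, concentrated by a steep power of U.
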Material /VehicleCosmetics/Boosts/Boost_Daikon/Materials/M_Snowman_Scarf
graph {
 properties {bAutomaticallySetUsageInEditor=false, bUsedWithNiagaraRibbons=true, blendMode=BLEND_Masked, shadingModel=MSM_Unlit, usedShadingModels=Unlit}
 nodes [
  n1 = Subtract(A=TextureCoordinate(), B=Constant() {r=0.5})
  Abs(None=@n1)
  n2 = TextureCoordinate() {coordinateIndex=1}
  n3 = Time()
  n4 = Add(
   A=Divide(A=ComponentMask(None=@n2) {r=true}, B=ScalarParameter() {defaultValue=0.5, group=End_Movement, parameterName=CountWaves2})
   B=Multiply(A=ScalarParameter() {defaultValue=7, group=End_Movement, parameterName=Speed}, B=OneMinus(None=@n3))
  )
  n5 = ComponentMask(None=TextureCoordinate()) {r=true}
  n6 = ScalarParameter() {defaultValue=0.5, group=End_Movement, parameterName=PowerMiddle2}
  n7 = Multiply(
   A=Multiply(A=Sine(None=@n4), B=ScalarParameter() {defaultValue=5, group=End_Movement, parameterName=Strenght})
   B=Multiply(A=Power(Base=@n5, Exp=@n6), B=ScalarParameter() {defaultValue=0.5, group=End_Movement, parameterName=MultPower2})
  )
  n8 = MaterialFunctionCall(X=Constant(), Y=@n7, Z=@n7) {materialFunction=&MakeFloat3}
  n9 = MaterialFunctionCall(
   WidthBase=ScalarParameter() {defaultValue=1, group=End_Movement, parameterName=Base}
   WidthTip=ScalarParameter() {defaultValue=1, group=End_Movement, parameterName=Tip}
  ) {materialFunction=&SplineThicken}
  n10 = ComponentMask(None=TextureCoordinate()) {r=true}
  n11 = Constant() {r=10}
  n12 = Multiply(A=OneMinus(None=Time()), B=ScalarParameter() {defaultValue=7, group=WPO, parameterName=Speed})
  n13 = ScalarParameter() {defaultValue=1, parameterName=WPO_TileAmt}
  n14 = TextureCoordinate()
  n15 = Divide(A=ComponentMask(None=Multiply(A=@n13, B=@n14)) {r=true}, B=ScalarParameter() {defaultValue=1, group=WPO, parameterName=tile})
  n16 = Add(A=@n12, B=@n15)
  n17 = Add(A=@n15, B=@n12)
  n18 = Sine(None=@n17)
  n19 = MaterialFunctionCall(X=Cosine(None=@n16) {period=0.5}, Y=@n18, Z=@n18) {materialFunction=&MakeFloat3}
  n20 = OneMinus(None=Cosine(None=@n17))
  n21 = MaterialFunctionCall(X=OneMinus(None=Sine(None=@n16) {period=0.75}), Y=@n20, Z=@n20) {materialFunction=&MakeFloat3}
  MaterialExpressionConstantClamp_1 = Saturate(None=LinearInterpolate(A=Sine(None=@n17) {period=2}, B=Cosine(None=@n16) {period=2.5}, Alpha=Constant() {r=0.5}))
  n22 = Multiply(
   A=Multiply(
    A=LinearInterpolate(A=@n19.Result, B=@n21.Result, Alpha=@MaterialExpressionConstantClamp_1)
    B=ScalarParameter() {defaultValue=1, group=WPO, parameterName=Intensity}
   )
   B=ScalarParameter() {defaultValue=1, group=WPO, parameterName=Amount}
  )
  n23 = ComponentMask(None=TextureCoordinate()) {r=true}
  MaterialExpressionConstantClamp_0 = Saturate(None=Multiply(A=Sine(None=@n23) {period=2}, B=Constant() {r=1}))
  n24 = Multiply(A=ConstantBiasScale(None=Transform(None=@n22) {transformType=TRANSFORM_Local}) {scale=0.25}, B=@MaterialExpressionConstantClamp_0)
  n25 = Time()
  n26 = Add(
   A=ComponentMask(None=TextureCoordinate() {coordinateIndex=1}) {r=true}
   B=Multiply(A=ScalarParameter() {defaultValue=2, parameterName=Twist}, B=OneMinus(None=@n25))
  )
  n27 = Sine(None=@n26) {period=2}
  n28 = MaterialFunctionCall(X=Cosine(None=@n26) {period=2}, Y=@n27, Z=@n27) {materialFunction=&MakeFloat3}
  n29 = ScalarParameter() {defaultValue=2, parameterName=Pinch}
  n30 = ComponentMask(None=Panner(Coordinate=TextureCoordinate() {coordinateIndex=1, uTiling=0.8}) {speedX=0.2}) {r=true}
  MaterialExpressionConstantClamp_3 = Saturate(None=ConstantBiasScale(None=Multiply(A=@n29, B=@n30)) {bias=0.4, scale=0.35})
  n31 = Multiply(
   A=Multiply(A=Transform(None=@n28.Result) {transformType=TRANSFORM_Local}, B=Constant() {r=20})
   B=Multiply(A=@MaterialExpressionConstantClamp_3, B=ComponentMask(None=@n1) {g=true})
  )
  Add(
   A=Multiply(A=Add(A=@n8.Result, B=@n9.WorldPositionOffset), B=Power(Base=@n10, Exp=@n11))
   B=Transform(None=Add(A=@n24, B=@n31) {constB=0}) {transformSourceType=TRANSFORMSOURCE_Local}
  )
  Constant() {r=1}
  MaterialFunctionCall(Emissive=NamedRerouteUsage()) {materialFunction=&MF_FX_EmissiveAdjustments}
  n32 = TextureSample() {
   samplerSource=SSM_Clamp_WorldGroupSettings
   samplerType=SAMPLERTYPE_LinearColor
   texture=/VehicleCosmetics/Boosts/Boost_Daikon/Textures/T_VCVFX_Boost_Daikon_Scarf_Masks_RGB
  }
  n33 = ParticleColor()
  NamedRerouteDeclaration(None=Multiply(A=@n32.R, B=@n33.A)) {name=Opacity, nodeColor=<rgba 1 0 0.0234 1>}
  n34 = TextureSample(UVs=TextureCoordinate() {vTiling=0.3}) {samplerType=SAMPLERTYPE_Normal, texture=/DealershipSharedLibrary/FX/Textures/Normal/T_VCVFX_FabricPillow_N}
  n35 = Constant() {r=4}
  n36 = TextureSample(UVs=TextureCoordinate() {uTiling=3.5}) {texture=/VehicleCosmetics/Boosts/Boost_Daikon/Textures/T_VCVFX_Boost_Daikon_Scarf_Snowman_D}
  n37 = VectorParameter() {parameterName=ColorWhiteScarf, defaultValue=<rgba 0.9819 1 0.8579 1>}
  n38 = Multiply(A=Desaturation(None=Multiply(A=@n34.RGB, B=@n35)), B=Add(A=Multiply(A=@n36.RGB, B=@n32.R), B=Multiply(A=@n32.G, B=@n37.RGB)))
  NamedRerouteDeclaration(None=Multiply(A=Multiply(A=@n38) {constB=2}, B=@n33.RGB)) {name=Color, nodeColor=<rgba 0 0.3437 1 1>}
  NamedRerouteUsage()
  Reroute(None=@n32.R)
 ]
}
